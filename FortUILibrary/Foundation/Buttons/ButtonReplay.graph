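# ButtonReplay is a Fortnite UI icon-and-text button based on IconTextButtonSpectatorBase, with an input action widget below and a right-side named slot.
# Input changes swap controller or mouse/keyboard styles, and construction or action changes also refresh text, preferring the action's display text over ButtonText.
# Hover and selection events recolor the Background border.
# PreConstruct applies alignment, text and icon, collapsing the image when its brush equals IconTextButton's default.
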WidgetBlueprint /FortUILibrary/Foundation/Buttons/ButtonReplay
graphs [
 ActivateButton {nodes [ActivateButton() {is=title K2Node_FunctionEntry}]}
 EventGraph {nodes [
   n1 = "On Input Method Changed (InputActionWidget)"!() -> (OutputDelegate, then, bUsingGamepad:bool) {is=title K2Node_ComponentBoundEvent}
   > UpdateStyle!(execute, self, UsingGamepad:bool=@n1.bUsingGamepad) -> (then)
   Event On Hovered()
   > SetBrushColor(self=Get Background() -> (Background:Border Object Reference), InBrushColor=Get HoveredColor() -> (HoveredColor:Linear Color Structure))
   n2 = Event SetText!() -> (OutputDelegate, then, Text:"Text (by ref)") {is=title K2Node_Event}
   > n3 = Set ButtonText(ButtonText:Text=@n2.Text) -> (Output_Get:Text)
   n4 = Get CenterButtonTextWidget() -> (CenterButtonTextWidget:Common Text Object Reference)
   > "SetText (Text)"(self=@n4, InText=@n3)
   > SetVisibility(self=@n4, InVisibility=Select(Option 0=SelfHitTestInvisible, Option 1=Collapsed, Index=TextIsEmpty(InText=@n2.Text)))
   n5 = Event SetTabLabelInfo!() -> (OutputDelegate, then, TabLabelInfo:"Fort Tab Button Label Info Structure (by ref)") {is=title K2Node_Event}
   > n6 = Sequence()
   n7 = Break Fort Tab Button Label Info!(FortTabButtonLabelInfo:"Fort Tab Button Label Info Structure (by ref)"=@n5.TabLabelInfo) -> (DisplayName:Text, IconBrush:Slate Brush Structure, IconResource:Soft Object Reference) {is=title K2Node_BreakStruct}
   > SetText!(execute, self:Icon Text Button Spectator Base Object Reference, Text:"Text (by ref)"=@n7) -> (then)
   Set Icon!(execute=@n6.then_1, self, IconBrush:Slate Brush Structure=@n7.IconBrush) -> (then) {is=title K2Node_CallFunction}
   Event OnCurrentTextStyleChanged() {is=title K2Node_Event}
   > UpdateTextStyle()
   Event On Unhovered()
   > SetBrushColor(self=Get Background(), InBrushColor=Get DefaultColor() -> (DefaultColor:Linear Color Structure))
   Event OnTriggeredInputActionChanged!() -> (OutputDelegate, then, NewTriggeredAction:"Data Table Row Handle Structure (by ref)") {is=title K2Node_Event}
   n8 = Get CommonInputSubsystem!() -> (ReturnValue:Common Input Object Reference) {is=title K2Node_GetBlueprintContextEx}
   > UpdateTextAndStyle!(execute, self, bUsingGamepad:bool="Equal (Enum)"(A=GetCurrentInputType(self=@n8), B=Gamepad)) -> (then)
   Event Construct()
   > InitializeButton()
   n9 = Get CommonInputSubsystem()
   > UpdateTextAndStyle(bUsingGamepad="Equal (Enum)"(A=GetCurrentInputType(self=@n9), B=Gamepad))
   n10 = Event PreConstruct()
   > n11 = Sequence!(execute) -> (then_0, then_1, then_2:Exec, then_3:Exec)
   "SetText (Text)"(execute=@n11.then_2, self=Get CenterButtonTextWidget(), InText=Get ButtonText() -> (ButtonText:Text))
   > UpdateTextStyle()
   UpdateContentAlignment(execute=@n11.then_0)
   Branch(execute=@n11.then_3, Condition=@n10.IsDesignTime)
   n12 = Get RightExtraContentSlot() -> (RightExtraContentSlot:Named Slot Object Reference)
   > SetVisibility(
    self=@n12
    InVisibility=Select(Option 0=Collapsed, Option 1=SelfHitTestInvisible, Index=HasAnyChildren!(self:Panel Widget Object Reference=@n12) -> (ReturnValue:bool))
   )
   n13 = Get IconBrush() -> (IconBrush:Slate Brush Structure)
   SetBrush(execute=@n11.then_1, self=Get Image() -> (Image:Image Object Reference), InBrush=@n13)
   n14 = Get Class Defaults!(Class:Object Class Reference=&/IconTextButton.IconTextButton_C) -> (
    ButtonText:Text
    IconBrush:Slate Brush Structure
    ControllerInputStyle:Common Button Style Class Reference
    MouseKeyboardStyle:Common Button Style Class Reference
    ContentAlignment:EHorizontalAlignment Enum
    bMouseKeyboardStyleSet:bool=~
    OverrideButtonText:Text
    ButtonTextJustification:ETextJustify Enum
    bDisplayAllCaps:bool=~
    PressProgress:double=~
    bIgnoreInputActionWidgetText:bool=~
    FocusChangedEvent:Multicast Delegate
    Font:Slate Font Info Structure
    UseLightInput:bool=~
   ) {is=title K2Node_GetClassDefaults}
   > SetVisibility(
    self=Get Image()
    InVisibility=Select(
     Option 0=Visible
     Option 1=Collapsed
     Index="Equal (SlateBrush)"!(A:"Slate Brush Structure (by ref)"=@n13, B:"Slate Brush Structure (by ref)"=@n14.IconBrush) -> (ReturnValue:bool)
    )
   )
   Event On Selected()
   > SetBrushColor(self=Get Background(), InBrushColor=Get SelectedColor() -> (SelectedColor:Linear Color Structure))
   Event On Deselected()
   > SetBrushColor(self=Get Background(), InBrushColor=Get DefaultColor())
   Event OnActionProgress!() -> (OutputDelegate, then, HeldPercent:float) {is=title K2Node_Event}
   Event OnActionComplete() {is=title K2Node_Event}
  ]}
 InitializeButton {nodes [
   InitializeButton() {is=title K2Node_FunctionEntry}
   > Set MouseKeyboardStyle(MouseKeyboardStyle:Common Button Style Class Reference=Get Style() -> (Style:Common Button Style Class Reference)) -> (Output_Get:Common Button Style Class Reference)
   > n1 = Is Valid(InputObject=GetBrushResource!(Brush:"Slate Brush Structure (by ref)"=Get IconBrush()) -> (ReturnValue:Object Reference))
   SetVisibility(execute=@n1.Is Not Valid, self=Get Image(), InVisibility=Collapsed)
   SetBrush(execute=@n1.Is Valid, self=Get Image(), InBrush=Get IconBrush())
  ]}
 Set Icon {nodes [
   n1 = Set Icon!() -> (then, IconBrush:Slate Brush Structure) {is=title K2Node_FunctionEntry}
   > n2 = Set IconBrush(IconBrush:Slate Brush Structure=@n1) -> (Output_Get:Slate Brush Structure)
   > SetBrush(self=Get Image(), InBrush=@n2)
  ]}
 SetContentAlignment {nodes [
   n1 = SetContentAlignment!() -> (then, ContentAlignment:EHorizontalAlignment Enum=HAlign_Fill) {is=title K2Node_FunctionEntry}
   > Set ContentAlignment(ContentAlignment:EHorizontalAlignment Enum=@n1) -> (Output_Get:EHorizontalAlignment Enum=HAlign_Fill)
   > UpdateContentAlignment()
  ]}
 SetControllerStyle {nodes [
   n1 = SetControllerStyle!() -> (then, ControllerInputStyle:Common Button Style Class Reference) {is=title K2Node_FunctionEntry}
   > Set ControllerInputStyle(ControllerInputStyle:Common Button Style Class Reference=@n1) -> (Output_Get:Common Button Style Class Reference)
   n2 = Get CommonInputSubsystem()
   > UpdateStyle(UsingGamepad="Equal (Enum)"(A=GetCurrentInputType(self=@n2), B=Gamepad))
  ]}
 SetMouseKeyboardStyle {nodes [
   n1 = SetMouseKeyboardStyle!() -> (then, ControllerInputStyle:Common Button Style Class Reference) {is=title K2Node_FunctionEntry}
   > Set MouseKeyboardStyle(MouseKeyboardStyle=@n1)
   n2 = Get CommonInputSubsystem()
   > UpdateStyle(UsingGamepad="Equal (Enum)"(A=GetCurrentInputType(self=@n2), B=Gamepad))
  ]}
 UpdateContentAlignment {nodes [
   UpdateContentAlignment() {is=title K2Node_FunctionEntry}
   > n1 = Cast To BorderSlot!(
    execute
    Object=GetContentSlot!(self:Content Widget Object Reference=Get ContentBorder() -> (ContentBorder:Border Object Reference)) -> (ReturnValue:Panel Slot Object Reference)
   ) -> (then, CastFailed, AsBorder Slot:Border Slot Object Reference) {is=title K2Node_DynamicCast}
   > SetHorizontalAlignment(
    self:Border Slot Object Reference=@n1
    InHorizontalAlignment=Get ContentAlignment() -> (ContentAlignment:EHorizontalAlignment Enum=HAlign_Fill)
   )
  ]}
 UpdateStyle {nodes [
   n1 = UpdateStyle!() -> (then, UsingGamepad:bool=~) {is=title K2Node_FunctionEntry}
   > n2 = Branch(Condition=@n1)
   > n3 = SetStyle!(
    execute
    self:Common Button Base Object Reference
    InStyle:Common Button Style Class Reference=Get ControllerInputStyle() -> (ControllerInputStyle:Common Button Style Class Reference)
   ) -> (then)
   Branch(execute=[@n3.then, @n4.then], Condition=Get ShouldHideIcon() -> (ShouldHideIcon:bool))
   > SetVisibility(self=Get InputActionWidget() -> (InputActionWidget:Common Action Widget Object Reference), InVisibility=Collapsed)
   n4 = SetStyle(execute=@n2.else, InStyle=Get MouseKeyboardStyle() -> (MouseKeyboardStyle:Common Button Style Class Reference))
  ]}
 UpdateText {nodes [
   UpdateText() {is=title K2Node_FunctionEntry}
   n1 = GetDisplayText!(self:Common Action Widget Object Reference=Get InputActionWidget()) -> (ReturnValue:Text)
   n2 = TextIsEmpty(InText=@n1)
   > SetText(Text=Select(Option 0:Text=Get ButtonText(), Option 1:Text=@n1, Index=NOT Boolean(A=@n2)) -> (ReturnValue:Text))
  ]}
 UpdateTextAndStyle {nodes [
   n1 = UpdateTextAndStyle!() -> (then, bUsingGamepad:bool=~) {is=title K2Node_FunctionEntry}
   > UpdateStyle(UsingGamepad=@n1) {is=title K2Node_CallFunction}
   > UpdateTextStyle()
   > UpdateText() {is=title K2Node_CallFunction}
  ]}
 UpdateTextStyle {nodes [
   UpdateTextStyle() {is=title K2Node_FunctionEntry}
   > SetStyle(
    self:Common Text Object Reference=Get CenterButtonTextWidget()
    InStyle:Common Text Style Class Reference=GetCurrentTextStyleClass!(self:Common Button Base Object Reference) -> (ReturnValue:Common Text Style Class Reference)
   )
  ]}
]
variables [
 ButtonText {type=string, default=""}
 ContentAlignment {type=EHorizontalAlignment, default=HAlign_Fill}
 ControllerInputStyle {type=Class, default=null}
 DefaultColor {type=LinearColor, default=<rgba 0.0176 0.0232 0.0331 0.5>}
 HoveredColor {type=LinearColor, default=<rgba 1 0.4125 0 1>}
 IconBrush {
  type=SlateBrush
  default {
   bIsDynamicallyLoaded=false
   drawAs=Image
   imageType=NoImage
   mirroring=NoMirror
   resourceName=None
   resourceObject=null
   tiling=NoTile
   imageSize=<40 40>
   margin=<ltrb 0 0 0 0>
   outlineSettings {
    bUseBrushTransparency=false
    roundingType=HalfHeightRadius
    width=0
    color {colorUseRule=UseColor_Specified, specifiedColor=<rgba 0 0 0 0>}
    cornerRadii=<0 0 0 1>
   }
   tintColor {colorUseRule=UseColor_Specified, specifiedColor=<rgba 1 1 1 1>}
   uVRegion {bIsValid=false, max=<0 0>, min=<0 0>}
  }
 }
 MouseKeyboardStyle {type=Class, default=/FortUILibrary/Foundation/Buttons/ButtonStyles/Primary/ButtonStyle-Primary-M.ButtonStyle-Primary-M_C}
 SelectedColor {type=LinearColor, default=<rgba 1 0.4125 0 1>}
 ShouldHideIcon {type=boolean}
]
widgets {
 info {inheritedWidgetCount=1, parentClass=/Script/FortniteUI.IconTextButtonSpectatorBase, rootWidgetClass=&Border, widgetCount=9}
 widgets [
  ContentBorder {
   widgetClassPath=&Border
   bIsVariable=true
   widget=ContentBorder
   properties {
    horizontalAlignment=HAlign_Center
    verticalAlignment=VAlign_Center
    background {
     bIsDynamicallyLoaded=false
     drawAs=NoDrawType
     imageType=NoImage
     mirroring=NoMirror
     resourceName=None
     resourceObject=null
     tiling=NoTile
     imageSize=<16 16>
     margin=<ltrb 0 0 0 0>
     outlineSettings {
      bUseBrushTransparency=false
      roundingType=HalfHeightRadius
      width=0
      color {colorUseRule=UseColor_Specified, specifiedColor=<rgba 0 0 0 0>}
      cornerRadii=<0 0 0 0>
     }
     tintColor {colorUseRule=UseColor_Specified, specifiedColor=<rgba 1 1 1 1>}
     uVRegion {bIsValid=false, max=<0 0>, min=<0 0>}
    }
    padding=<ltrb 0 0 0 0>
   }
  }
  VerticalBox_0 {
   widgetClassPath=&VerticalBox
   parent=ContentBorder
   widget=VerticalBox_0
   slot {class=BorderSlot, properties {horizontalAlignment=HAlign_Center, verticalAlignment=VAlign_Center, padding=<ltrb 0 0 0 0>}}
  }
  HorizontalBox_1 {
   widgetClassPath=&HorizontalBox
   parent=VerticalBox_0
   widget=HorizontalBox_1
   properties {renderTransform {angle=0, scale=<0 0>, shear=<0 0>, translation=<0 0>}}
   slot {class=VerticalBoxSlot}
  }
  Background {
   widgetClassPath=&Border
   bIsVariable=true
   parent=HorizontalBox_1
   widget=Background
   properties {
    horizontalAlignment=HAlign_Center
    verticalAlignment=VAlign_Center
    background {
     bIsDynamicallyLoaded=false
     drawAs=NoDrawType
     imageType=NoImage
     mirroring=NoMirror
     resourceName=None
     resourceObject=null
     tiling=NoTile
     imageSize=<0 32>
     margin=<ltrb 0 0 0 0>
     outlineSettings {
      bUseBrushTransparency=false
      roundingType=HalfHeightRadius
      width=0
      color {colorUseRule=UseColor_Specified, specifiedColor=<rgba 0 0 0 0>}
      cornerRadii=<0 0 0 0>
     }
     tintColor {colorUseRule=UseColor_Specified, specifiedColor=<rgba 1 1 1 1>}
     uVRegion {bIsValid=false, max=<0 0>, min=<0 0>}
    }
    brushColor=<rgba 0.0176 0.0232 0.0331 0.498>
    padding=<ltrb 0 0 0 0>
   }
   slot {class=HorizontalBoxSlot}
  }
  HorizontalBox_0 {
   widgetClassPath=&HorizontalBox
   parent=Background
   widget=HorizontalBox_0
   properties {renderTransform {angle=0, scale=<0 1>, shear=<0 0>, translation=<0 0>}}
   slot {class=BorderSlot, properties {horizontalAlignment=HAlign_Center, verticalAlignment=VAlign_Center, padding=<ltrb 0 0 0 0>}}
  }
  Image {
   widgetClassPath=&Image
   bIsVariable=true
   parent=HorizontalBox_0
   widget=Image
   properties {brush {
     bIsDynamicallyLoaded=false
     drawAs=NoDrawType
     imageType=NoImage
     mirroring=NoMirror
     resourceName=None
     resourceObject=null
     tiling=NoTile
     imageSize=<0 0>
     margin=<ltrb 0 0 0 0>
     outlineSettings {
      bUseBrushTransparency=false
      roundingType=HalfHeightRadius
      width=0
      color {colorUseRule=UseColor_Specified, specifiedColor=<rgba 0 0 0 0>}
      cornerRadii=<0 0 0 0>
     }
     tintColor {colorUseRule=UseColor_Specified, specifiedColor=<rgba 1 1 1 1>}
     uVRegion {bIsValid=false, max=<0 0>, min=<0 0>}
    }}
   slot {class=HorizontalBoxSlot}
  }
  CenterButtonTextWidget {
   widgetClassPath=&CommonTextBlock
   bIsVariable=true
   parent=HorizontalBox_0
   widget=CenterButtonTextWidget
   properties {
    justification=Center
    scrollStyle=&TextScrollStyle-Base_C
    style=/FortUILibrary/Styles/Typography/Foundation/TextStyle-Header-M.TextStyle-Header-M_C
    font {
     bForceMonospaced=false
     bMaterialIsStencil=false
     fontMaterial=null
     fontObject=&Burbank
     letterSpacing=0
     monospacedWidth=1
     size=32
     skewAmount=0
     typefaceFontName=Burbank-Black
     outlineSettings {
      bApplyOutlineToDropShadows=false
      bMiteredCorners=false
      bSeparateFillAlpha=false
      outlineMaterial=null
      outlineSize=0
      outlineColor=<rgba 0 0 0 1>
     }
    }
    shadowOffset=<0 0>
   }
   slot {class=HorizontalBoxSlot, properties {verticalAlignment=VAlign_Center}}
  }
  RightExtraContentSlot {
   widgetClassPath=&NamedSlot
   bIsVariable=true
   parent=HorizontalBox_1
   widget=RightExtraContentSlot
   properties {bExposeOnInstanceOnly=true}
   slot {class=HorizontalBoxSlot}
  }
  InputActionWidget {
   widgetClassPath=&CommonActionWidget
   bInherited=true
   bIsVariable=true
   parent=VerticalBox_0
   widget=InputActionWidget
   slot {class=VerticalBoxSlot, properties {horizontalAlignment=HAlign_Center, verticalAlignment=VAlign_Center}}
  }
 ]
}
