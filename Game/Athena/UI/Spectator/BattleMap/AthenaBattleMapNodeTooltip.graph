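# Tooltip widget for nodes on the spectator battle map, showing a node's title, time and description on a translucent dark panel.
# A bottom row shows "Look At" and "Scrub To" hints with left and right mouse icons, plus a gamepad "Scrub To" button bound to the BattleMapScrubTo replay input action.
# The Blueprint is layout only, with empty event stubs; key widgets are inherited from the C++ AthenaBattleMapNodeTooltip parent.
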WidgetBlueprint /Game/Athena/UI/Spectator/BattleMap/AthenaBattleMapNodeTooltip
graphs [EventGraph {nodes [
   Event PreConstruct()
   Event Construct()
   Event Tick()
  ]}]
variables []
widgets {
 info {inheritedWidgetCount=6, parentClass=/Script/FortniteUI.AthenaBattleMapNodeTooltip, rootWidgetClass=&Border, widgetCount=14}
 widgets [
  Border_75 {
   widgetClassPath=&Border
   widget=Border_75
   properties {background {
     bIsDynamicallyLoaded=false
     drawAs=Image
     imageType=NoImage
     mirroring=NoMirror
     resourceName=None
     resourceObject=null
     tiling=NoTile
     imageSize=<32 32>
     margin=<ltrb 0 0 0 0>
     outlineSettings {
      bUseBrushTransparency=false
      roundingType=HalfHeightRadius
      width=0
      color {colorUseRule=UseColor_Specified, specifiedColor=<rgba 0 0 0 0>}
      cornerRadii=<0 0 0 0>
     }
     tintColor {colorUseRule=UseColor_Specified, specifiedColor=<rgba 0 0 0 0.5>}
     uVRegion {bIsValid=false, max=<0 0>, min=<0 0>}
    }}
  }
  VerticalBox_27 {widgetClassPath=&VerticalBox, parent=Border_75, widget=VerticalBox_27, slot {class=BorderSlot}}
  HorizontalBox_0 {widgetClassPath=&HorizontalBox, parent=VerticalBox_27, widget=HorizontalBox_0, slot {class=VerticalBoxSlot}}
  TextBlock_NodeTitle {
   widgetClassPath=&CommonTextBlock
   bInherited=true
   parent=HorizontalBox_0
   widget=TextBlock_NodeTitle
   properties {
    style=/FortUILibrary/Styles/Typography/Foundation/TextStyle-Base-S-B.TextStyle-Base-S-B_C
    text=Title
    font {
     bForceMonospaced=false
     bMaterialIsStencil=false
     fontMaterial=null
     fontObject=&NotoSans
     letterSpacing=0
     monospacedWidth=1
     size=18
     skewAmount=0
     typefaceFontName=NotoSans-Bold
     outlineSettings {
      bApplyOutlineToDropShadows=false
      bMiteredCorners=false
      bSeparateFillAlpha=false
      outlineMaterial=null
      outlineSize=0
      outlineColor=<rgba 0 0 0 1>
     }
    }
    shadowOffset=<0 0>
   }
   slot {class=HorizontalBoxSlot, properties {verticalAlignment=VAlign_Center, size {sizeRule=Fill, value=1}}}
  }
  TextBlock_NodeTime {
   widgetClassPath=&CommonTextBlock
   bInherited=true
   parent=HorizontalBox_0
   widget=TextBlock_NodeTime
   properties {
    style=&TextStyle-Base-S-B_C
    text=00:00
    font {
     bForceMonospaced=false
     bMaterialIsStencil=false
     fontMaterial=null
     fontObject=&NotoSans
     letterSpacing=0
     monospacedWidth=1
     size=18
     skewAmount=0
     typefaceFontName=NotoSans-Bold
     outlineSettings {
      bApplyOutlineToDropShadows=false
      bMiteredCorners=false
      bSeparateFillAlpha=false
      outlineMaterial=null
      outlineSize=0
      outlineColor=<rgba 0 0 0 1>
     }
    }
    shadowOffset=<0 0>
   }
   slot {class=HorizontalBoxSlot, properties {horizontalAlignment=HAlign_Right, verticalAlignment=VAlign_Center, padding=<ltrb 15 0 0 0>}}
  }
  TextBlock_NodeDescription {
   widgetClassPath=&CommonTextBlock
   bInherited=true
   parent=VerticalBox_27
   widget=TextBlock_NodeDescription
   properties {
    style=/FortUILibrary/Styles/Typography/Foundation/TextStyle-Base-S.TextStyle-Base-S_C
    text=Description
    font {
     bForceMonospaced=false
     bMaterialIsStencil=false
     fontMaterial=null
     fontObject=&NotoSans
     letterSpacing=0
     monospacedWidth=1
     size=18
     skewAmount=0
     typefaceFontName=NotoSans-Regular
     outlineSettings {
      bApplyOutlineToDropShadows=false
      bMiteredCorners=false
      bSeparateFillAlpha=false
      outlineMaterial=null
      outlineSize=0
      outlineColor=<rgba 0 0 0 1>
     }
    }
    shadowOffset=<0 0>
   }
   slot {class=VerticalBoxSlot, properties {verticalAlignment=VAlign_Center}}
  }
  HorizontalBox_177 {
   widgetClassPath=&HorizontalBox
   parent=VerticalBox_27
   widget=HorizontalBox_177
   slot {class=VerticalBoxSlot, properties {verticalAlignment=VAlign_Bottom, size {sizeRule=Fill, value=1}}}
  }
  Widget_Selection {
   widgetClassPath=&HorizontalBox
   bInherited=true
   parent=HorizontalBox_177
   widget=Widget_Selection
   slot {class=HorizontalBoxSlot, properties {horizontalAlignment=HAlign_Left, verticalAlignment=VAlign_Center}}
  }
  LeftClickButton {
   widgetClassPath=&Image
   parent=Widget_Selection
   widget=LeftClickButton
   properties {brush {
     bIsDynamicallyLoaded=false
     drawAs=Image
     imageType=NoImage
     mirroring=NoMirror
     resourceName=None
     resourceObject=/Game/UI/Fort_Quickbar/LeftMouse_ON
     tiling=NoTile
     imageSize=<24 37.5>
     margin=<ltrb 0 0 0 0>
     outlineSettings {
      bUseBrushTransparency=false
      roundingType=HalfHeightRadius
      width=0
      color {colorUseRule=UseColor_Specified, specifiedColor=<rgba 0 0 0 0>}
      cornerRadii=<0 0 0 0>
     }
     tintColor {colorUseRule=UseColor_Specified, specifiedColor=<rgba 1 1 1 1>}
     uVRegion {bIsValid=false, max=<0 0>, min=<0 0>}
    }}
   slot {class=HorizontalBoxSlot}
  }
  LeftClickText {
   widgetClassPath=&CommonTextBlock
   parent=Widget_Selection
   widget=LeftClickText
   properties {
    style=&TextStyle-Base-S-B_C
    text=Look At
    font {
     bForceMonospaced=false
     bMaterialIsStencil=false
     fontMaterial=null
     fontObject=&NotoSans
     letterSpacing=0
     monospacedWidth=1
     size=18
     skewAmount=0
     typefaceFontName=NotoSans-Bold
     outlineSettings {
      bApplyOutlineToDropShadows=false
      bMiteredCorners=false
      bSeparateFillAlpha=false
      outlineMaterial=null
      outlineSize=0
      outlineColor=<rgba 0 0 0 1>
     }
    }
    shadowOffset=<0 0>
   }
   slot {class=HorizontalBoxSlot, properties {verticalAlignment=VAlign_Center, padding=<ltrb 15 0 20 0>}}
  }
  Widget_ScrubbingGamepad {
   widgetClassPath=/FortUILibrary/Foundation/Buttons/IconTextButtonReplayHorizontal
   bInherited=true
   bIsVariable=true
   parent=HorizontalBox_177
   widget=Widget_ScrubbingGamepad
   properties {
    buttonText=Scrub To
    onButtonBaseClicked=(&self:WidgetTree.Widget_ScrubbingGamepad.HandleOnButtonClicked)
    onButtonBaseDoubleClicked=(&self:WidgetTree.Widget_ScrubbingGamepad.HandleOnButtonDoubleClicked)
    onButtonBaseHovered=(&self:WidgetTree.Widget_ScrubbingGamepad.HandleOnButtonHovered)
    onButtonBaseUnhovered=(&self:WidgetTree.Widget_ScrubbingGamepad.HandleOnButtonUnhovered)
    onSelectedChangedBase=(&self:WidgetTree.Widget_ScrubbingGamepad.HandleOnSelectedChanged)
    style=/FortUILibrary/Foundation/Buttons/ButtonStyles/ButtonStyle-TextOnlyBase_S-B.ButtonStyle-TextOnlyBase_S-B_C
    iconBrush {
     bIsDynamicallyLoaded=false
     drawAs=Image
     imageType=NoImage
     mirroring=NoMirror
     resourceName=None
     resourceObject=null
     tiling=NoTile
     imageSize=<0 0>
     margin=<ltrb 0 0 0 0>
     outlineSettings {
      bUseBrushTransparency=false
      roundingType=HalfHeightRadius
      width=0
      color {colorUseRule=UseColor_Specified, specifiedColor=<rgba 0 0 0 0>}
      cornerRadii=<0 0 0 1>
     }
     tintColor {colorUseRule=UseColor_Specified, specifiedColor=<rgba 1 1 1 1>}
     uVRegion {bIsValid=false, max=<0 0>, min=<0 0>}
    }
    triggeringInputAction {dataTable=/FortUILibrary/Input/ReplayInputActionDataTable, rowName=BattleMapScrubTo}
   }
   slot {class=HorizontalBoxSlot, properties {horizontalAlignment=HAlign_Right, verticalAlignment=VAlign_Center, size {sizeRule=Fill, value=1}}}
  }
  Widget_Scrubbing {
   widgetClassPath=&HorizontalBox
   bInherited=true
   parent=HorizontalBox_177
   widget=Widget_Scrubbing
   slot {class=HorizontalBoxSlot, properties {horizontalAlignment=HAlign_Right, size {sizeRule=Fill, value=1}}}
  }
  RightClickText {
   widgetClassPath=&CommonTextBlock
   parent=Widget_Scrubbing
   widget=RightClickText
   properties {
    style=&TextStyle-Base-S-B_C
    text=Scrub To
    font {
     bForceMonospaced=false
     bMaterialIsStencil=false
     fontMaterial=null
     fontObject=&NotoSans
     letterSpacing=0
     monospacedWidth=1
     size=18
     skewAmount=0
     typefaceFontName=NotoSans-Bold
     outlineSettings {
      bApplyOutlineToDropShadows=false
      bMiteredCorners=false
      bSeparateFillAlpha=false
      outlineMaterial=null
      outlineSize=0
      outlineColor=<rgba 0 0 0 1>
     }
    }
    shadowOffset=<0 0>
   }
   slot {class=HorizontalBoxSlot, properties {verticalAlignment=VAlign_Center, padding=<ltrb 20 0 15 0>}}
  }
  RightClickButton {
   widgetClassPath=&Image
   parent=Widget_Scrubbing
   widget=RightClickButton
   properties {brush {
     bIsDynamicallyLoaded=false
     drawAs=Image
     imageType=NoImage
     mirroring=NoMirror
     resourceName=None
     resourceObject=/Game/UI/Fort_Quickbar/RightMouse_ON
     tiling=NoTile
     imageSize=<24 37.5>
     margin=<ltrb 0 0 0 0>
     outlineSettings {
      bUseBrushTransparency=false
      roundingType=HalfHeightRadius
      width=0
      color {colorUseRule=UseColor_Specified, specifiedColor=<rgba 0 0 0 0>}
      cornerRadii=<0 0 0 0>
     }
     tintColor {colorUseRule=UseColor_Specified, specifiedColor=<rgba 1 1 1 1>}
     uVRegion {bIsValid=false, max=<0 0>, min=<0 0>}
    }}
   slot {class=HorizontalBoxSlot, properties {horizontalAlignment=HAlign_Right}}
  }
 ]
}
