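# Widget Blueprint for the timeline control on the Athena spectator battle map, built on the native AthenaBattleMapTimelineWidget class.
# An overlay holds the inherited Forward and Backward replay buttons, both slightly translucent.
# Each button is triggered by its own timeline input action from the replay input action data table.
# The event graph holds only unconnected PreConstruct, Construct and Tick events, so the Blueprint adds no logic of its own.
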WidgetBlueprint /Game/Athena/UI/Spectator/BattleMap/AthenaBattleMapTimelineWidget
graphs [EventGraph {nodes [
   Event PreConstruct()
   Event Construct()
   Event Tick()
  ]}]
variables []
widgets {
 info {inheritedWidgetCount=2, parentClass=/Script/FortniteUI.AthenaBattleMapTimelineWidget, rootWidgetClass=&Overlay, widgetCount=3}
 widgets [
  Overlay_4 {widgetClassPath=&Overlay, widget=Overlay_4}
  Button_Forward {
   widgetClassPath=/FortUILibrary/Foundation/Buttons/ButtonReplay
   bInherited=true
   bIsVariable=true
   parent=Overlay_4
   widget=Button_Forward
   properties {
    navigation=&self:WidgetTree.Button_Forward.WidgetNavigation_0
    onButtonBaseClicked=(&self:WidgetTree.Button_Forward.HandleOnButtonClicked)
    onButtonBaseDoubleClicked=(&self:WidgetTree.Button_Forward.HandleOnButtonDoubleClicked)
    onButtonBaseHovered=(&self:WidgetTree.Button_Forward.HandleOnButtonHovered)
    onButtonBaseUnhovered=(&self:WidgetTree.Button_Forward.HandleOnButtonUnhovered)
    onSelectedChangedBase=(&self:WidgetTree.Button_Forward.HandleOnSelectedChanged)
    renderOpacity=0.75
    triggeringInputAction {dataTable=/FortUILibrary/Input/ReplayInputActionDataTable, rowName=BattleMapTimelineForward}
   }
   slot {class=OverlaySlot}
  }
  Button_Backward {
   widgetClassPath=&ButtonReplay
   bInherited=true
   bIsVariable=true
   parent=Overlay_4
   widget=Button_Backward
   properties {
    navigation=&self:WidgetTree.Button_Backward.WidgetNavigation_0
    onButtonBaseClicked=(&self:WidgetTree.Button_Backward.HandleOnButtonClicked)
    onButtonBaseDoubleClicked=(&self:WidgetTree.Button_Backward.HandleOnButtonDoubleClicked)
    onButtonBaseHovered=(&self:WidgetTree.Button_Backward.HandleOnButtonHovered)
    onButtonBaseUnhovered=(&self:WidgetTree.Button_Backward.HandleOnButtonUnhovered)
    onSelectedChangedBase=(&self:WidgetTree.Button_Backward.HandleOnSelectedChanged)
    renderOpacity=0.75
    triggeringInputAction {dataTable=&ReplayInputActionDataTable, rowName=BattleMapTimelineBackward}
   }
   slot {class=OverlaySlot}
  }
 ]
}
